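# Sound cue for the first-person music of the Heartsign emote in Athena.
# A Mixer starts a one-shot intro wave and a looping wave together, feeding an Enveloper whose volume follows a shared external curve named as an eight-second loop fade-out.
# It uses the 1P emote music sound class and concurrency, keeps playing when silent, and sends to a music audio-analysis submix.
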SoundCue /Game/Athena/Sounds/Emotes/Heartsign/EmoteMusic_Heartsign_Enveloped
root=Enveloper_0
nodes [
 {
  id=Enveloper_0
  children [Mixer_0]
  properties {
   pitchCurve {editorCurveData {keys [{time=0}]}}
   volumeCurve {externalCurve=&/Emote_Loop_Env_8sec_FadeOut, editorCurveData {keys [<at 0 1>]}}
  }
 }
 {id=Mixer_0, children [WavePlayer_0,WavePlayer_1], properties {inputVolume [1,1]}}
 {id=WavePlayer_0, properties {soundWaveAssetPtr=&/Emote_Heartsign_v5_Intro}}
 {id=WavePlayer_1, properties {bLooping=true, soundWaveAssetPtr=&/Emote_Heartsign_v5_Loop}}
]
properties {
 maxDistance=2097152
 soundClassObject=/Game/Sounds/FortSoundClassesAndModes/Fort_Music_Emote_1P_PSM
 virtualizationMode=PlayWhenSilent
 volumeMultiplier=1
 concurrencySet [/Game/Sounds/Concurrency_Rules/Athena_EmoteMusic_1P_Conc]
 soundSubmixSends [{minSendLevel=0, soundSubmix=&Fortnite_AudioAnalysis_Music}]
}
